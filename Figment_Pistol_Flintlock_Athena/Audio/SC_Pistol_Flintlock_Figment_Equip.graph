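# Sound cue for the equip sound of the Figment flintlock pistol.
# A short fixed delay feeds a single wave player that reuses a six-shooter equip recording.
# It routes through the weapon foley sound class with the shared weapon reload attenuation.
# A concurrency override prevents new instances from starting once the limit is reached.
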SoundCue /Figment_Pistol_Flintlock_Athena/Audio/SC_Pistol_Flintlock_Figment_Equip
root=Delay_0
nodes [
 {id=Delay_0, children [WavePlayer_0], properties {delayMax=0.13, delayMin=0.13}}
 {id=WavePlayer_0, properties {soundWaveAssetPtr=&/SixShooter_Equip_01}}
]
properties {
 attenuationSettings=&Fort_Weapon_Reload_Att
 bOverrideConcurrency=true
 maxDistance=2250
 soundClassObject=&Fort_WeaponFoley
 volumeMultiplier=1
 concurrencyOverrides {resolutionRule=PreventNew}
}
